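# Material function in the F_MED_Mochi character body folder that toggles visibility of mesh elements marked by a vertex color value.
# It subtracts Threshold from saturated Alpha and feeds that raw difference as the hide toggle into the shared VertexColorMask function, alongside the channel and value inputs.
# A second output, isVisible, applies Ceil and Saturate to the difference, giving 1 only when Alpha exceeds Threshold.
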MaterialFunction /BRCosmetics/Characters/Player/Female/Medium/Bodies/F_MED_Mochi/Materials/MF_MakeVertexColorElementVisible
graph {nodes [
  Constant() {r=1}
  n1 = Constant() {r=0.3}
  n2 = Add(
   A=Subtract(B=FunctionInput(Preview=@n1) {inputName=Threshold, inputType=FunctionInput_Scalar}) {constA=0}
   B=Saturate(None=FunctionInput() {inputName=Alpha, inputType=FunctionInput_Scalar})
  )
  FunctionOutput(None=MaterialFunctionCall!(
    Hide Element Toggle=@n2
    VertexColorChannel=FunctionInput() {inputName=Vertex Color Channel}
    Vertex Value to hide=FunctionInput() {inputName=Vertex Color Value to Hide, inputType=FunctionInput_Scalar}
   ) {materialFunction=/Game/Packages/Fortress_SharedMaterials/Base_Material_Functions/VertexColorMask}.Result)
  FunctionOutput(None=Saturate(None=Ceil(None=@n2))) {outputName=isVisible}
  OneMinus(None=@n2)
 ]}
